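; ResolveGameplay UI director that shows homing rocket lock-on HUD scenes while that launcher is equipped.
; On BeginPlay it binds to the HUD context's weapon-equipped delegate and casts each new weapon to the homing rocket launcher interface.
; A successful cast adds the lock-on circle and lock-on scenes; a failed cast removes them.
; A boolean flag prevents repeated adds or removes, while overlap, tick and construction events are empty.
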(event EventBeginPlay
  (HUD|BindEventtoOnWeaponEquippedDelegate (BlueprintContext|GetFortHUDContext) (delegate Custom|OnWeaponEquipped)))

(event Custom|OnWeaponEquipped
  (params (NewWeapon :type Fort%20Weapon%20Object%20Reference) (PrevWeapon :type Fort%20Weapon%20Object%20Reference))
  (bind _baddedweaponscenes (Variables|Default|GetbAddedWeaponScenes))
  (Utilities|Casting|CastToB_Launcher_HomingRocket_Athena_Interface NewWeapon
    (:then
      (if (not _baddedweaponscenes)
        (DynamicUIDirector|AddScene "/ResolveGameplay/Items/Guns/HomingRocket/Widgets/HomingRocket_LockOnCircle_Scene.HomingRocket_LockOnCircle_Scene")
        (DynamicUIDirector|AddScene "/ResolveGameplay/Items/Guns/HomingRocket/Widgets/HomingRocket_LockOn_Scene.HomingRocket_LockOn_Scene")
        (Variables|Default|SetbAddedWeaponScenes true)))
    (:CastFailed
      (if _baddedweaponscenes
        (DynamicUIDirector|RemoveScene "/ResolveGameplay/Items/Guns/HomingRocket/Widgets/HomingRocket_LockOnCircle_Scene.HomingRocket_LockOnCircle_Scene")
        (DynamicUIDirector|RemoveScene "/ResolveGameplay/Items/Guns/HomingRocket/Widgets/HomingRocket_LockOn_Scene.HomingRocket_LockOn_Scene")
        (Variables|Default|SetbAddedWeaponScenes false)))))

(event Collision|EventActorBeginOverlap
  (params (OtherActor :type Actor%20Object%20Reference)))

(event EventTick
  (params (DeltaSeconds :type Float%20%28single-precision%29)))

(fn ConstructionScript)
